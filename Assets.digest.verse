# Copyright Epic Games, Inc. All Rights Reserved.
#################################################
# Generated Digest of Verse API
# DO NOT modify this manually!
# Generated from build: ++Fortnite+Release-36.30-CL-44367537
#################################################



    <#> M_Multiplier contain properties that were not reflected.
         Some parameters from M_Multiplier were skipped due to being an invalid identifier: 
             'Base Color'


    <#> M_Target contain properties that were not reflected.
         Some parameters from M_Target were skipped due to being an invalid identifier: 
             'Base Color'


using {/Verse.org/Native}
using {/Verse.org/Simulation}
using {/Verse.org/SceneGraph}
using {/Verse.org/Assets}
# Module import path: /playkigai.dig@fortnite.com/Minigame_IcicleThrowing/Materials
Materials := module:
    M_Multiplier<scoped {/playkigai.dig@fortnite.com/Minigame_IcicleThrowing}> := class<final><scoped {/playkigai.dig@fortnite.com/Minigame_IcicleThrowing}>(material):


    M_Target<scoped {/playkigai.dig@fortnite.com/Minigame_IcicleThrowing}> := class<final><scoped {/playkigai.dig@fortnite.com/Minigame_IcicleThrowing}>(material):


    M_Trasparent<scoped {/playkigai.dig@fortnite.com/Minigame_IcicleThrowing}> := class<final><scoped {/playkigai.dig@fortnite.com/Minigame_IcicleThrowing}>(material):


# Module import path: /playkigai.dig@fortnite.com/Minigame_IcicleThrowing/Meshes
Meshes := module:
    SM_Projectile<scoped {/playkigai.dig@fortnite.com/Minigame_IcicleThrowing}> := class<final><scoped {/playkigai.dig@fortnite.com/Minigame_IcicleThrowing}>(mesh_component):
        @editable
        var Material_0<public>:material = external {}

    SM_Projectile_asset<scoped {/playkigai.dig@fortnite.com/Minigame_IcicleThrowing}>:mesh = external {}

    SM_ScoreMultiplier<scoped {/playkigai.dig@fortnite.com/Minigame_IcicleThrowing}> := class<final><scoped {/playkigai.dig@fortnite.com/Minigame_IcicleThrowing}>(mesh_component):
        @editable
        var WorldGridMaterial<public>:material = external {}

    SM_ScoreMultiplier_asset<scoped {/playkigai.dig@fortnite.com/Minigame_IcicleThrowing}>:mesh = external {}

    SM_StartingPlate<scoped {/playkigai.dig@fortnite.com/Minigame_IcicleThrowing}> := class<final><scoped {/playkigai.dig@fortnite.com/Minigame_IcicleThrowing}>(mesh_component):


    SM_StartingPlate_asset<scoped {/playkigai.dig@fortnite.com/Minigame_IcicleThrowing}>:mesh = external {}

    SM_Target<scoped {/playkigai.dig@fortnite.com/Minigame_IcicleThrowing}> := class<final><scoped {/playkigai.dig@fortnite.com/Minigame_IcicleThrowing}>(mesh_component):
        @editable
        var WorldGridMaterial<public>:material = external {}

    SM_Target_asset<scoped {/playkigai.dig@fortnite.com/Minigame_IcicleThrowing}>:mesh = external {}

    SM_TargetCircle<scoped {/playkigai.dig@fortnite.com/Minigame_IcicleThrowing}> := class<final><scoped {/playkigai.dig@fortnite.com/Minigame_IcicleThrowing}>(mesh_component):
        @editable
        var WorldGridMaterial<public>:material = external {}

    SM_TargetCircle_asset<scoped {/playkigai.dig@fortnite.com/Minigame_IcicleThrowing}>:mesh = external {}

# Module import path: /playkigai.dig@fortnite.com/Minigame_IcicleThrowing/Prefabs
Prefabs := module:
    @import_as("/Minigame_IcicleThrowing/Prefabs/game_manager.game_manager_C")
    game_manager<scoped {/playkigai.dig@fortnite.com/Minigame_IcicleThrowing}> := class<final>(entity):


    game_manager_asset<scoped {/playkigai.dig@fortnite.com/Minigame_IcicleThrowing}>:entity_prefab = external {}

    @import_as("/Minigame_IcicleThrowing/Prefabs/projectile.projectile_C")
    projectile<scoped {/playkigai.dig@fortnite.com/Minigame_IcicleThrowing}> := class<final>(entity):


    projectile_asset<scoped {/playkigai.dig@fortnite.com/Minigame_IcicleThrowing}>:entity_prefab = external {}

    @import_as("/Minigame_IcicleThrowing/Prefabs/projectile_spawner.projectile_spawner_C")
    projectile_spawner<scoped {/playkigai.dig@fortnite.com/Minigame_IcicleThrowing}> := class<final>(entity):


    projectile_spawner_asset<scoped {/playkigai.dig@fortnite.com/Minigame_IcicleThrowing}>:entity_prefab = external {}

    @import_as("/Minigame_IcicleThrowing/Prefabs/score_multiplier.score_multiplier_C")
    score_multiplier<scoped {/playkigai.dig@fortnite.com/Minigame_IcicleThrowing}> := class<final>(entity):


    score_multiplier_asset<scoped {/playkigai.dig@fortnite.com/Minigame_IcicleThrowing}>:entity_prefab = external {}

    @import_as("/Minigame_IcicleThrowing/Prefabs/start_game.start_game_C")
    start_game<scoped {/playkigai.dig@fortnite.com/Minigame_IcicleThrowing}> := class<final>(entity):


    start_game_asset<scoped {/playkigai.dig@fortnite.com/Minigame_IcicleThrowing}>:entity_prefab = external {}

    @import_as("/Minigame_IcicleThrowing/Prefabs/target.target_C")
    target<scoped {/playkigai.dig@fortnite.com/Minigame_IcicleThrowing}> := class<final>(entity):


    target_asset<scoped {/playkigai.dig@fortnite.com/Minigame_IcicleThrowing}>:entity_prefab = external {}

    @import_as("/Minigame_IcicleThrowing/Prefabs/target_spawner.target_spawner_C")
    target_spawner<scoped {/playkigai.dig@fortnite.com/Minigame_IcicleThrowing}> := class<final>(entity):


    target_spawner_asset<scoped {/playkigai.dig@fortnite.com/Minigame_IcicleThrowing}>:entity_prefab = external {}

    @import_as("/Minigame_IcicleThrowing/Prefabs/ui_manager.ui_manager_C")
    ui_manager<scoped {/playkigai.dig@fortnite.com/Minigame_IcicleThrowing}> := class<final>(entity):


    ui_manager_asset<scoped {/playkigai.dig@fortnite.com/Minigame_IcicleThrowing}>:entity_prefab = external {}