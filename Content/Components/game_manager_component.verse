<#
@UUID: 4e374ba4-9501-49fc-a02d-d2640f085121
@Task: Make a component that will manage the game progression.
@Description: Implement a SceneGraph component that  will manage the game progression.
@Tags: Gameplay
@Export: true
#>

using { /Verse.org/SceneGraph }
using { /Verse.org/SceneGraph/KeyframedMovement }
using { /Verse.org/Simulation }
using { /Fortnite.com/Devices }

game_manager_component<public> := class<final_super>(component):
    var MaybeRedPlayer:?agent = false
    var MaybeBluePlayer:?agent = false

    var PlayerScore:[agent]float = map{}

    var MaybeUIManager:?entity = false

    @editable
    ScoreToWin:float = 10.0

    @editable
    MessageDevice:hud_message_device = hud_message_device{}
    @editable
    MutatorZone:mutator_zone_device = mutator_zone_device{}

    OnBeginSimulation<override>():void=
        (super:)OnBeginSimulation()
        if:
            Sim := Entity.GetSimulationEntity[]
            Targets := Sim.FindDescendantEntitiesWithComponent(target_component)
            UIManagerEntities := Sim.FindDescendantEntitiesWithComponent(ui_manager_component)
        then:
            for(UIManager : UIManagerEntities, Comp := UIManager.GetComponent[ui_manager_component]):
                set MaybeUIManager = option. UIManager
            MutatorZone.Disable()
    
    AddScore<private>(InAgent:agent, Score:float):void=
        if:
            CurrentScore := PlayerScore[InAgent]
            set PlayerScore[InAgent] = CurrentScore + Score

            RedPlayer := MaybeRedPlayer?
            RedScore := PlayerScore[RedPlayer]
            BluePlayer := MaybeBluePlayer?
            BlueScore := PlayerScore[BluePlayer]
            Comp := MaybeUIManager?.GetComponent[ui_manager_component]
        then:
            # Update UI
            Comp.UpdateTexts(RedScore,BlueScore,ScoreToWin)
            CheckWinner()
                 
    OnTargetSpawned<public>(Target:target_component):void=
        Target.Subscribe(AddScore)

    CheckWinner():void=
        if(RedPlayer := MaybeRedPlayer?, BluePlayer := MaybeBluePlayer?, RedScore := PlayerScore[RedPlayer], BlueScore := PlayerScore[BluePlayer]):
            if. RedScore >= ScoreToWin then DeclareWinner(RedPlayer)
            if. BlueScore >= ScoreToWin then DeclareWinner(BluePlayer)
    
    DeclareWinner(Agent:agent):void=
        MutatorZone.Disable()
        MessageDevice.SetText(WinnerMessage(Agent))
        MessageDevice.Show()
        ResetGame()

    ResetGame():void=
        if(Sim := Entity.GetSimulationEntity[]):
            Spawners := Sim.FindDescendantEntitiesWithComponent(target_spawner_component)
            StartGames := Sim.FindDescendantEntitiesWithComponent(start_game_component)
            UIManagers := Sim.FindDescendantEntitiesWithComponent(ui_manager_component)
            for(UIManagerEntity : UIManagers, UIManager := UIManagerEntity.GetComponent[ui_manager_component]):
                UIManager.Reset()
            for(SpawnerEntity : Spawners, Spawner := SpawnerEntity.GetComponent[target_spawner_component]):
                Spawner.Reset()
            for(StartGameEntity : StartGames, StartGame := StartGameEntity.GetComponent[start_game_component]):
                StartGame.Reset()


(GameManager:game_manager_component).Init(RedPlayer:agent,BluePlayer:agent):void=
    if:
        set GameManager.MaybeRedPlayer = option. RedPlayer
        set GameManager.MaybeBluePlayer = option. BluePlayer
        set GameManager.PlayerScore[RedPlayer] = 0.0
        set GameManager.PlayerScore[BluePlayer] = 0.0
    then:
        GameManager.MutatorZone.Enable()

WinnerMessage<localizes>(Winner:agent):message="{Winner} won!"



