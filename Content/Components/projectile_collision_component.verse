<#
@UUID: 48f0a538-60ee-4271-a313-f8ab4e9216cd
@Task: Make a component that will handle collisions for a projectile.
@Description: Implement a SceneGraph component that will handle collisions for a projectile.
@Tags: Collisions, Gameplay
@Export: true
#>

using { /Verse.org/SceneGraph }
using { /Verse.org/SceneGraph/KeyframedMovement }
using { /Verse.org/Simulation }

projectile_collision_component<public> := class<final_super>(component):

    var MaybeOwningAgent:?agent = false
    var Multiplier: float = 1.0

    OnBeginSimulation<override>():void=
        (super:)OnBeginSimulation()
        if(Mesh := Entity.GetComponent[mesh_component]):
            Mesh.EntityEnteredEvent.Subscribe(OnEntityEntered)


    OnEntityEntered(InEntity:entity):void=
        if(MultiplierComp := InEntity.GetComponent[score_multiplier_component], MovementComp := InEntity.GetParent[].GetComponent[score_multiplier_movement_component]):
            if(MovementComp.IsActive?):
                MovementComp.Disable()
                set Multiplier = MultiplierComp.Multiplier
        if(TargetComp := InEntity.GetComponent[target_component]):
            TargetComp.TriggerHit(Self)
            Entity.RemoveFromParent()

(ProjColl:projectile_collision_component).SetAgent(InAgent:agent):void=
    set ProjColl.MaybeOwningAgent = option. InAgent