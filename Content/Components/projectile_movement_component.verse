<#
@UUID: 5baf571a-1532-48f0-a2e4-bd6e2eac89b3
@Task: Make a component that will handle the movement of a projectile.
@Description: Implement a SceneGraph component that will manage the movement of a projectile.
@Tags: KeyframedMovement, Gameplay
@Export: true
#>

using { /Fortnite.com/Characters }
using { /Verse.org/SceneGraph }
using { /Verse.org/SceneGraph/KeyframedMovement }
using { /Verse.org/Simulation }
using { /Verse.org/SpatialMath }
using { /UnrealEngine.com/Temporary/SpatialMath }

projectile_movement_component<public> := class<final_super>(component):
    @editable
    DefaultEasingFunction<public>:easing_function = ease_in_out_cubic_bezier_easing_function{}
    @editable
    PlaybackMode<public>:keyframed_movement_playback_mode = oneshot_keyframed_movement_playback_mode{}
    @editable
    ProjectileRange:float = 5000.0   
    @editable
    ProjectileSpeed:float = 2.0
    var MaybeAgent:?agent = false


    Init<public>(InAgent:agent):void=
        set MaybeAgent = option. InAgent

    # Construct a single keyframe which animates between the Source and Destination entity using the given easing function over a set duration.
    ConstructKeyframe<private>(SpawningAgent:agent, EasingFunction:easing_function, Duration:float)<transacts><decides>:?[]keyframed_movement_delta=
        if:
            Fort := SpawningAgent.GetFortCharacter[]
            Player_Pos:(/Verse.org/SpatialMath:)vector3= (/Verse.org/SpatialMath:)vector3{
                Forward := Fort.GetTransform().Translation.X, 
                Left := -Fort.GetTransform().Translation.Y,
                Up := Fort.GetTransform().Translation.Z
                }
            Player_View_Rot := Fort.GetViewRotation()

            Player_Push_Dir:(/Verse.org/SpatialMath:)vector3 = (/Verse.org/SpatialMath:)vector3{
                Forward := Player_View_Rot.GetLocalForward().X,
                Left := -Player_View_Rot.GetLocalForward().Y,
                Up := Player_View_Rot.GetLocalForward().Z
                } * 400.0

            Player_End_Push_Dir:(/Verse.org/SpatialMath:)vector3 = (/Verse.org/SpatialMath:)vector3{
                Forward := Player_View_Rot.GetLocalForward().X,
                Left := -Player_View_Rot.GetLocalForward().Y,
                Up := Player_View_Rot.GetLocalForward().Z
            } * ProjectileRange
        then:
            option. array:
                keyframed_movement_delta:
                    Transform := (/Verse.org/SpatialMath:)transform{
                        Translation := (Player_Pos + Player_End_Push_Dir) - (Player_Pos + Player_Push_Dir)
                        Rotation := (/Verse.org/SpatialMath:)MakeRotationFromYawPitchRollDegrees(0.0,0.0,0.0)
                    }
                    Duration := Duration
                    Easing := EasingFunction
        else:
            return false
    

    MoveProjectile<public>(?Ease:easing_function = DefaultEasingFunction)<suspends>: void=
        if:
            Agent := MaybeAgent?
            Fort := Agent.GetFortCharacter[]
            Player_View_Rot := Fort.GetViewRotation()
            KeyframedMovementComponent := Entity.GetComponent[keyframed_movement_component]
        then:
            #Init Projectile Transform
            Entity.SetGlobalTransform(
                (/Verse.org/SpatialMath:)transform: 
                    Translation := (/Verse.org/SpatialMath:)vector3{
                        Forward := Fort.GetTransform().Translation.X + (Player_View_Rot.GetLocalForward().X * 200),
                        Left := -Fort.GetTransform().Translation.Y + (-Player_View_Rot.GetLocalForward().Y * 200),
                        Up := Fort.GetTransform().Translation.Z + (Player_View_Rot.GetLocalForward().Z * 200)
                        }
                    Scale := (/Verse.org/SpatialMath:)vector3{Forward := 1.0, Left := 1.0, Up := 1.0}
                )
            var Keyframes:[]keyframed_movement_delta = array{}
            if: 
                Keyframe := ConstructKeyframe[Agent, Ease, ProjectileSpeed]?
                set Keyframes += Keyframe

            KeyframedMovementComponent.SetKeyframes(Keyframes, PlaybackMode)
            KeyframedMovementComponent.Play()
            KeyframedMovementComponent.KeyframeReachedEvent.Await()
            KeyframedMovementComponent.Pause()
            #At the end of the movement we destroy the projectile
            Entity.RemoveFromParent()