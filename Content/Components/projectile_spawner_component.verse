<#
@UUID: fc22f9a5-0086-411d-83b5-4ad7e9f679b3
@Task: Make a component that will handle the spawning of projectiles.
@Description: Implement a SceneGraph component that will handle the spawning of projectiles.
@Tags: Gameplay
@Export: true
#>

using { /Fortnite.com/Characters }
using { /Fortnite.com/Devices }
using { /Fortnite.com/Playspaces }
using { /Fortnite.com/UI }
using { /Verse.org/Colors }
using { /Verse.org/Random }
using { /Verse.org/SceneGraph }
using { /Verse.org/Simulation }
using { /Verse.org/SpatialMath }
using { /UnrealEngine.com/Temporary/SpatialMath }
using {Prefabs}

projectile_spawner_component<public> := class<final_super>(component):

    @editable
    ProjectileCooldown:float = 1.0
    @editable
    MaybeInputTrigger:?input_trigger_device = false
    var PlayerCooldown:[agent]logic = map{}

    OnBeginSimulation<override>():void=
        (super:)OnBeginSimulation()
        if:
            Sim := Entity.GetSimulationEntity[]
            Playspace := Sim.GetPlayspaceForEntity[]
            Players := Playspace.GetPlayers()
            InputTrigger := MaybeInputTrigger?
        then:
            Playspace.ParticipantAddedEvent().Subscribe(OnParticipantAdded)
            for(Player : Players):
                InputTrigger.Register(Player)
                if. set PlayerCooldown[Player] = true
            InputTrigger.PressedEvent.Subscribe(OnPressed)

    OnParticipantAdded(Agent :agent):void=
        if(InputTrigger := MaybeInputTrigger?):
            InputTrigger.Register(Agent)
            if. set PlayerCooldown[Agent] = true

    OnPressed(Agent:agent):void=
        if(PlayerCooldown[Agent]?):
            SpawnProjectile(Agent)
            if. set PlayerCooldown[Agent] = false
            spawn. HandleCooldown(Agent, ProjectileCooldown)

    HandleCooldown(Agent:agent, Cooldown:float)<suspends>:void=
        Sleep(Cooldown)
        if. set PlayerCooldown[Agent] = true
        
    SpawnProjectile<public>(SpawningAgent:agent):void=
        if:
            Sim := Entity.GetSimulationEntity[]
            ProjectilePrefab := projectile{}
            ProjectileMovementComp := ProjectilePrefab.GetComponent[projectile_movement_component]
            ProjectileCollisionComp := ProjectilePrefab.GetComponent[projectile_collision_component]
        then:
            ProjectileMovementComp.Init(SpawningAgent)
            ProjectileCollisionComp.SetAgent(SpawningAgent)
            Sim.AddEntities(array. ProjectilePrefab)
            spawn{ProjectileMovementComp.MoveProjectile()}


