<#
@UUID: b72e145c-8c98-4aeb-bca1-64e2ab335188
@Task: Make a component that will handle a score multiplier.
@Description: Implement a SceneGraph component that will increase the score if the projectile enter in collision with it
@Tags: KeyframedMovement, Gameplay
@Export: true
#>

using { /Verse.org }
using { /Verse.org/Native }
using { /Verse.org/SceneGraph }
using { /Verse.org/SpatialMath }
using { /Verse.org/Simulation }

# A Verse-authored component that can be added to entities
score_multiplier_component<public> := class<final_super>(component):
    
    var Multiplier<public>:float = 2.0