<#
@UUID: b72e145c-8c98-4aeb-bca1-64e2ab335188
@Task: Make a component that will handle a score multiplier.
@Description: Implement a SceneGraph component that will increase the score if the projectile enter in collision with it
@Tags: KeyframedMovement, Gameplay
@Export: true
#>

using { /Verse.org }
using { /Verse.org/Native }
using { /Verse.org/SceneGraph }
using { /Verse.org/SpatialMath }
using { /Verse.org/Simulation }

score_multiplier_movement_component<public> := class<final_super>(component):

    @editable
    ActivationTime:float = 5.0

    @editable:
        Categories := array{StartRotationCategory}
    StartRotationLeft:float = 0.0
    @editable:
        Categories := array{StartRotationCategory}
    StartRotationUp:float = 0.0
    @editable:
        Categories := array{StartRotationCategory}
    StartRotationForward:float = 0.0

    @editable:
        Categories := array{TargetRotationCategory}
    TargetRotationLeft:float = 0.0
    @editable:
        Categories := array{TargetRotationCategory}
    TargetRotationUp:float = 0.0
    @editable:
        Categories := array{TargetRotationCategory}
    TargetRotationForward:float = 0.0

    var IsActive:logic = false

    OnBeginSimulation<override>():void=
        (super:)OnBeginSimulation()
        spawn. ActivationTimer()

    ActivationTimer()<suspends>:void=
        Sleep(ActivationTime)
        Activate()

    Activate()<suspends>:void=
        FromVector:vector3 = vector3{Left := StartRotationLeft, Up :=StartRotationUp, Forward := StartRotationForward}
        ToVector:vector3 = vector3{Left := TargetRotationLeft, Up :=TargetRotationUp, Forward := TargetRotationForward}
        Duration:float = 0.25
        var Elapsed:float = 0.0
        loop:
            set Elapsed += 0.01
            T:float = Elapsed/Duration
            LerpedVector := Lerp(FromVector, ToVector, T)
            Entity.SetGlobalTransform(
                transform:
                    Translation := Entity.GetGlobalTransform().Translation
                    Rotation := MakeRotationFromEulerDegrees(LerpedVector.Left,LerpedVector.Up,LerpedVector.Forward)
                    Scale := Entity.GetGlobalTransform().Scale
                )
            Sleep(0.01)
            if (Elapsed >= Duration):
                Entity.SetGlobalTransform(
                transform:
                    Translation := Entity.GetGlobalTransform().Translation
                    Rotation := MakeRotationFromEulerDegrees(ToVector.Left,ToVector.Up,ToVector.Forward)
                    Scale := Entity.GetGlobalTransform().Scale
                )
                break
        set IsActive = true;

    Disable<public>():void=
        set IsActive = false
        FromVector:vector3 = vector3{Left := StartRotationLeft, Up :=StartRotationUp, Forward := StartRotationForward}
        Entity.SetGlobalTransform(
            transform:
                Translation := Entity.GetGlobalTransform().Translation
                Rotation := MakeRotationFromEulerDegrees(FromVector.Left,FromVector.Up,FromVector.Forward)
                Scale := Entity.GetGlobalTransform().Scale
            )
        spawn. ActivationTimer()


StartRotationCategory<localizes>:message="StartRotation"
TargetRotationCategory<localizes>:message="TargetRotation"