<#
@UUID: 58316346-1e36-43ec-ba20-0d74a02ab1ce
@Task: Make a component that acts as a target that can be hit.
@Description: Implement a SceneGraph component that acts as a target that can be hit by players' projectiles. 
@Tags: Gameplay, Collisions
@Export: true
#>

using { /Verse.org }
using { /Verse.org/Native }
using { /Verse.org/Random }
using { /Verse.org/SceneGraph }
using { /Verse.org/Simulation }
using { /Verse.org/SpatialMath }
using { /Verse.org/SceneGraph/KeyframedMovement }

# Custom function type for TargetHitSubscribable.
target_hit_function := type{_(:agent, :float):void}

# Class for custom subscribable.
target_hit_subscribable := class:
    var Functions<private>:[]target_hit_function = array{}

    # Subscribe-like function.
    # Needs to be wrapped by target_component.
    # Accepts functions of type target_hit_function, which will be run during Dispatch().
    AddFunction(InFunction:target_hit_function):void=
        set Functions += array{InFunction}


    # Call all functions registered.
    Dispatch(InAgent:agent, Score:float):void=
        for (Function : Functions):
            Function(InAgent, Score)


target_component<public> := class<final_super>(component):
    
    @editable
    var Score:float = 1.0

    TargetHitSubscribable: target_hit_subscribable = target_hit_subscribable{}

    RemovedFromScene : event() = event(){}

    var MaybePlayerOne<public>: ?agent = false
    var MaybePlayerTwo<public>: ?agent = false
    var MaybePostPhysics: ?cancelable = false

    OnBeginSimulation<override>(): void =
        (super:)OnBeginSimulation()


    OnRemovingFromScene<override>(): void =
        if (PlayerOne := MaybePlayerOne?):
            Print("Awarded point to Player One.")

        RemovedFromScene.Signal()

    OnEndSimulation<override>(): void =
        if (PostPhysics := MaybePostPhysics?) then PostPhysics.Cancel()

# Function called by projectiles when colliding with a target, Dispatch the TargetHitSubscribable
(TargetComp:target_component).TriggerHit<public>(ProjectileCollision: projectile_collision_component): void=
    if (InAgent := ProjectileCollision.MaybeOwningAgent?, Sim := TargetComp.Entity.GetSimulationEntity[]):
        TargetComp.TargetHitSubscribable.Dispatch(InAgent, TargetComp.Score * ProjectileCollision.Multiplier)
        if(ParentEntity := TargetComp.Entity.GetParent[]):
            Spawners := Sim.FindDescendantComponents(target_spawner_component)
            for(Spawner : Spawners) do Spawner.DestroyTarget(ParentEntity)

# Extention of the target prefab that allows to Subscribe to the target_component TargetHitSubscribable
(Target: target_component).Subscribe<public>(InFunction : target_hit_function):void=
    Target.TargetHitSubscribable.AddFunction(InFunction)

target_component_end<public> := class<final_super>(component):