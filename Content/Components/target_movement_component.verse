<#
@UUID: 080cef5a-d916-48f0-bd3e-66def2153bd8
@Task: Make a component that moves a target between two points.
@Description: Implement a SceneGraph component that that moves a target between two points.
@Tags: KeyframedMovement, Gameplay
@Export: true
#>

using { /Verse.org }
using { /Verse.org/Native }
using { /Verse.org/Random }
using { /Verse.org/SceneGraph }
using { /Verse.org/Simulation }
using { /Verse.org/SpatialMath }
using { /Verse.org/SceneGraph/KeyframedMovement }

target_movement_component<public> := class<final_super>(component):

    var MaybeStartPos:?entity = false
    var MaybeEndPos:?entity = false
    @editable
    MinMoveDuration:float = 1.0
    @editable
    MaxMoveDuration:float = 5.0

    var IsMoving:logic = false

    CreateKeyFrame(StartPos:vector3, EndPos:vector3, InRotation:rotation): keyframed_movement_delta =
        keyframed_movement_delta:
            Transform := transform:
                Translation := vector3{
                    Forward := GetRandomFloat(StartPos.Forward, EndPos.Forward) - Entity.GetGlobalTransform().Translation.Forward,
                    Left := GetRandomFloat(StartPos.Left, EndPos.Left) - Entity.GetGlobalTransform().Translation.Left,
                    Up := GetRandomFloat(StartPos.Up, EndPos.Up) -  Entity.GetGlobalTransform().Translation.Up
                    }
                Scale := vector3{}
            Duration := GetRandomFloat(MinMoveDuration,MaxMoveDuration)
            Easing := linear_easing_function{}
    
    MoveTarget()<suspends>:void=
        loop:
            if:
                not IsMoving?
                MovementComponent := Entity.GetComponent[keyframed_movement_component]
                StartPos := MaybeStartPos?.GetGlobalTransform().Translation
                EndPos := MaybeEndPos?.GetGlobalTransform().Translation
                InRot := MaybeStartPos?.GetGlobalTransform().Rotation
            then:
                set IsMoving = true
                Keyframe := CreateKeyFrame(StartPos,EndPos,InRot)
                MovementComponent.SetKeyframes(array{Keyframe}, oneshot_keyframed_movement_playback_mode{})
                MovementComponent.Play()
                MovementComponent.KeyframeReachedEvent.Await()
                set IsMoving = false
            Sleep(0.1)

(TargetMovement:target_movement_component).Init(StartPosition:entity,EndPosition:entity):void=
    set TargetMovement.MaybeStartPos = option. StartPosition
    set TargetMovement.MaybeEndPos = option. EndPosition