<#
@UUID: 91b1b0f4-ebc8-4c84-8d10-45e1c3accaf4
@Task: Make a component that spawns and initialize a target entity.
@Description: Implement a SceneGraph component that spawns and initialize a target entity.
@Tags: Gameplay
@Export: true
#>

using { /Verse.org }
using { /Verse.org/Native }
using { /Verse.org/Random }
using { /Verse.org/SceneGraph }
using { /Verse.org/Simulation }
using { /Verse.org/SpatialMath }
using {Prefabs}

target_spawner_component<public> := class<final_super>(component):
    @editable
    var MaybeStartPos:?entity = false
    @editable
    var MaybeEndPos:?entity = false
    @editable
    SpawnDelay:float = 2.0

   var MaybeTargetEntityRef:?entity = false

    SpawnTarget():void=
        if:
            StartPos := MaybeStartPos?
            EndPos := MaybeEndPos?
            not MaybeTargetEntityRef?
            Sim := Entity.GetSimulationEntity[]
            GameManagers := Sim.FindDescendantComponents(game_manager_component)
            Target := target{}
            MoveComp := Target.GetComponent[target_movement_component]
            TargetComps := Target.FindDescendantComponents(target_component)
            var SpawnPos:vector3 = vector3{}
            if. GetRandomInt(0,1) > 0 then set SpawnPos = StartPos.GetGlobalTransform().Translation else set SpawnPos = EndPos.GetGlobalTransform().Translation
            set MaybeTargetEntityRef = option. Target
            Target.SetGlobalTransform(
                transform:
                    Translation := SpawnPos,
                    Rotation := StartPos.GetLocalTransform().Rotation,
                    Scale := vector3{Forward := 2.0, Left := 2.0, Up := 2.0}
                )
        then:
            Sim.AddEntities(array{Target})
            MoveComp.Init(StartPos,EndPos)
            spawn. MoveComp.MoveTarget()
            for(TargetComp : TargetComps):
                for(GameManager:GameManagers):
                    GameManager.OnTargetSpawned(TargetComp)


    CheckNewTargetSpawn()<suspends>:void=
        Sleep(SpawnDelay)      
        SpawnTarget()

(Spawner:target_spawner_component).Reset():void=
    if(EntityRef := Spawner.MaybeTargetEntityRef?):
            EntityRef.RemoveFromParent()
            set Spawner.MaybeTargetEntityRef = false
            
(Spawner:target_spawner_component).DestroyTarget(TargetToDestroy:entity):void=
    if(TargetEntityRef := Spawner.MaybeTargetEntityRef?, TargetEntityRef = TargetToDestroy):
        TargetEntityRef.RemoveFromParent()
        set Spawner.MaybeTargetEntityRef = false
        spawn. Spawner.CheckNewTargetSpawn()

(Spawner:target_spawner).Init():void=
    if(Comp := Spawner.GetComponent[target_spawner_component]):
        Comp.SpawnTarget()


