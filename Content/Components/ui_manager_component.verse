<#
@UUID: c45caad1-c97a-4e19-8907-16768bd658ff
@Task: A component to manage the UI of both players.
@Description: Implement a SceneGraph component that will manage the UI of bith players.
@Tags: UI
@Export: true
#>

using { /Verse.org }
using { /Fortnite.com/UI }
using { /Verse.org/Native }
using { /Verse.org/SceneGraph }
using { /Verse.org/Simulation }
using { /UnrealEngine.com/Temporary/UI }
using { /UnrealEngine.com/Temporary/SpatialMath }

ui_manager_component<public> := class<final_super>(component):

    var MaybeRedPlayer: ?agent = false
    var MaybeBluePlayer: ?agent = false
    var PlayersUI: [agent]agent_ui = map{}

    # Remove UI from Agent.
    RemoveUI<public>(Agent:agent):void=
        if(AgentUI := PlayersUI[Agent], PlayerUI := GetPlayerUI[player[Agent]]):
            PlayerUI.RemoveWidget(AgentUI.Canvas)

    # Init UI for Agent.
    InitUI(BluePlayer:agent, RedPlayer:agent, MaxScore:float):void=
        Print("UI Init")
        set MaybeBluePlayer = option. BluePlayer
        set MaybeRedPlayer = option. RedPlayer
        if. set PlayersUI[RedPlayer] = agent_ui{}
        if. set PlayersUI[BluePlayer] = agent_ui{}

        # Generating and adding UI to both players.
        GenerateUI()
        UpdateTexts(0.0,0.0,MaxScore)

    # Generate and add UI to both Red and Blue players.
    GenerateUI():void=
        if:
            BluePlayer := MaybeBluePlayer?, 
            RedPlayer := MaybeRedPlayer?
            BluePlayerUI := GetPlayerUI[player[BluePlayer]], 
            RedPlayerUI := GetPlayerUI[player[RedPlayer]]
            BlueUI := PlayersUI[BluePlayer], 
            RedUI := PlayersUI[RedPlayer]
        then:
            RedPlayerUI.AddWidget(RedUI.Canvas, player_ui_slot{InputMode := ui_input_mode.None})
            BluePlayerUI.AddWidget(BlueUI.Canvas, player_ui_slot{InputMode := ui_input_mode.None})

    
    # Update the text for both players.
    UpdateTexts<public>(RedScore:float,BlueScore:float, MaxScore:float):void=
        if:
            RedPlayer := MaybeRedPlayer?, 
            BluePlayer := MaybeBluePlayer?
            BlueUI := PlayersUI[BluePlayer], 
            RedUI := PlayersUI[RedPlayer]
        then:
            BlueUI.UpdateUI(RedPlayer, RedScore, BluePlayer, BlueScore, MaxScore)
            RedUI.UpdateUI(RedPlayer, RedScore, BluePlayer, BlueScore, MaxScore)

    OnEndSimulation<override>():void=
        if:
            BluePlayer := MaybeBluePlayer?
            RedPlayer := MaybeRedPlayer?
        then:
            RemoveUI(BluePlayer)
            RemoveUI(RedPlayer)

(UIManager:ui_manager_component).Reset():void=
    Print("Removing UI")
    if(RedPlayer := UIManager.MaybeRedPlayer?, BluePlayer := UIManager.MaybeBluePlayer?):
        UIManager.RemoveUI(RedPlayer)
        UIManager.RemoveUI(BluePlayer)
        set UIManager.MaybeRedPlayer = false
        set UIManager.MaybeBluePlayer = false

agent_ui := class():
    RedPlayerTextBlock: text_block = text_block{}
    BluePlayerTextBlock: text_block = text_block{}
    var Canvas<internal>: canvas = canvas{}
    block:
        set Canvas = canvas:
            Slots := array{
                canvas_slot{
                    Widget := RedPlayerTextBlock
                    Anchors := anchors{Minimum := vector2{X := 0.4, Y := 0.0}, Maximum := vector2{X := 0.4, Y := 0.0}}
                    Offsets := margin{Top := 75.0}
                    SizeToContent := true
                },
                canvas_slot{
                    Widget := BluePlayerTextBlock
                    Anchors := anchors{Minimum := vector2{X := 0.4, Y := 0.0}, Maximum := vector2{X := 0.4 , Y := 0.0}}
                    Offsets := margin{Top := 115.0}
                    SizeToContent := true
                }
            }

    UpdateUI<public>(RedAgent:agent, RedScore:float,BlueAgent:agent, BlueScore:float, MaxScore:float):void=
        RedPlayerTextBlock.SetText(PlayerScoreMessage(RedAgent, RedScore,MaxScore))
        BluePlayerTextBlock.SetText(PlayerScoreMessage(BlueAgent, BlueScore,MaxScore))

PlayerScoreMessage<localizes>(Agent:agent,Score:float,MaxScore:float):message="{Agent}: {Score}/{MaxScore}"