<#
@UUID: 9f2942f9-a685-4761-9b3d-7f2e6b4f9fe3
@Task: Editor Entites Prefab Setup
@Description: A file containing detailed information about how to create and setup the entity prefabs needed for the game. 
@Tags: Editor, Setup.
@Export: true
#>

<#
In-Editor game_manager prefab setup:
game_manager
    - transform_component
    - game_manager_component
#>

<#
In-Editor projectile prefab setup:
projectile
    - projectile_movement_component
    - SM_Projectile (mesh_component)
    - keyframed_movement_component
    - transform_component
    - projectile_collision_component
#>

<#
In-Editor projectile_spawner prefab setup:
projectile_spawner
    - transform_component
    - projectile_spawner_component
#>

<#
In-Editor score_multiplier prefab setup:
score_multiplier
    - transform_component
    - SM_ScoreMultiplier (mesh_component)
    - score_multiplier_movement_component
    - target (entity)
        - transform_component
        - SM_TargetCircle (mesh_component)
        - score_multiplier_component
#>

<#
In-Editor start_game prefab setup:
start_game
    - start_game_component
    - transform_component
    - blue_interactable (entity)
        - transform_component
        - mesh_component
        - game_interactable_component
    - blue_starting_position
        - transform_component
        - SM_StartingPlate (mesh_component)
    - red_interactable (entity)
        - transform_component
        - mesh_component
        - game_interactable_component
    - red_starting_position
        - transform_component
        - SM_StartingPlate (mesh_component)
#>

<#
In-Editor target prefab setup:
target
    - transform_component
    - SM_ScoreMultiplier (mesh_component)
    - target_movement_component
    - keyframed_movement_component
    - target_circle (entity)
        - transform_component
        - SM_TargetCircle (mesh_component)
        - target_component
#>

<#
In-Editor target_spawner prefab setup:
target_spawner
    - target_spawner_component
    - end_pos (entity)
        - transform_component
        - mesh_component
    - start_pos (entity)
        - transform_component
        - mesh_component
#>

<#
In-Editor ui_manager prefab setup:
ui_manager
    - transform_component
    - ui_manager_component
#>
