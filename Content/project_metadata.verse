<#
@UUID: 19f839b1-9e4b-4320-82ea-d778051beb0d
@Task: Icicle Throwing.
@Description: A minigame where two players challenge each other shooting at moving targets and scoring points with the addition of score multipliers. The first player to reach the target score wins. Targets appear at random and the player throw projectiles to hit them.
@Tags: Minigame, Scoring, Projectiles, Multiplayer
@Export: true
#>